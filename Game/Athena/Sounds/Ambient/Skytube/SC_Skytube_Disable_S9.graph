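# Athena ambient sound cue that plays the Skytube disable sound, routed to the Fort_Ambient_Athena sound class.
# A Delay node holds the single wave player back by up to a fraction of a second.
# Overridden attenuation gives it distance falloff, a wide stereo spread and listener focus that lowers off-focus volume.
# Overridden concurrency allows one instance at a time, stopping the farthest before refusing new ones.
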SoundCue /Game/Athena/Sounds/Ambient/Skytube/SC_Skytube_Disable_S9
root=Delay_0
nodes [
 {id=Delay_0, children [WavePlayer_0], properties {delayMax=0.3}}
 {id=WavePlayer_0, properties {soundWaveAssetPtr=&/Skytube_Disable_01}}
]
properties {
 bOverrideAttenuation=true
 bOverrideConcurrency=true
 maxDistance=6000
 soundClassObject=&Fort_Ambient_Athena
 volumeMultiplier=1
 attenuationOverrides {
  bEnableFocusInterpolation=true
  bEnableListenerFocus=true
  falloffDistance=6000
  focusDistanceScale=2
  focusReleaseInterpSpeed=5
  lPFFrequencyAtMax=6000
  lPFFrequencyAtMin=6000
  nonFocusVolumeAttenuation=0.5
  occlusionInterpolationTime=1
  occlusionLowPassFilterFrequency=2000
  occlusionTraceChannel=ECC_GameTraceChannel11
  occlusionVolumeAttenuation=0.5
  stereoSpread=2000
  attenuationShapeExtents {x=0}
 }
 concurrencyOverrides {maxCount=1, resolutionRule=StopFarthestThenPreventNew, volumeScaleAttackTime=0.1}
}
