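# Sound cue that plays a Skytube (wind conduit) entry sound for players, classed as Athena ambient audio.
# An IsLocal switch sends unset and non-local cases through the third-person Skytube attenuation, and local playback through an attenuation node with no assigned settings.
# Both branches share one equally weighted random pick of four WindConduit_Enter waves.
# A root modulator at default settings wraps the switch.
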SoundCue /Game/Athena/Sounds/Ambient/Skytube/SC_Skytube_Enter_Player_S9
root=Modulator_0
nodes [
 {
  id=Attenuation_0
  children [Random_4]
  properties {
   attenuationSettings=&/Skytube_3P_Att
   attenuationOverrides {falloffDistance=10000, stereoSpread=1024, attenuationShapeExtents {x=0}}
  }
 }
 {id=Attenuation_1, children [Random_4]}
 {id=Modulator_0, children [Switch_0]}
 {id=Random_4, children [WavePlayer_17, WavePlayer_19, WavePlayer_21, WavePlayer_23], properties {weights [1,1,1,1]}}
 {id=Switch_0, children [Attenuation_0, Attenuation_0, Attenuation_1], properties {intParameterName=IsLocal}}
 {id=WavePlayer_17, properties {soundWaveAssetPtr=/Game/Athena/Sounds/Ambient/Skytube/Rev1/WindConduit_Enter_01}}
 {id=WavePlayer_19, properties {soundWaveAssetPtr=/Game/Athena/Sounds/Ambient/Skytube/Rev1/WindConduit_Enter_03}}
 {id=WavePlayer_21, properties {soundWaveAssetPtr=/Game/Athena/Sounds/Ambient/Skytube/Rev1/WindConduit_Enter_06}}
 {id=WavePlayer_23, properties {soundWaveAssetPtr=/Game/Athena/Sounds/Ambient/Skytube/Rev1/WindConduit_Enter_07}}
]
properties {
 maxDistance=2097152
 soundClassObject=&Fort_Ambient_Athena
 volumeMultiplier=0.6
 attenuationOverrides {falloffDistance=6000, stereoSpread=2000, attenuationShapeExtents {x=0}}
}
