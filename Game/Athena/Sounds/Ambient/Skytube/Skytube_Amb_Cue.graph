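# Ambient SoundCue for the Athena Skytube, routed to the Athena ambient sound class.
# A Mixer layers two looping waves, a wind conduit ambience and an energy layer.
# Cue-level attenuation overrides add distance falloff, a low-pass filter, wide stereo spread, and occlusion on a custom trace channel that muffles and lowers volume.
# Interpolated listener focus makes the sound much quieter outside the focus cone.
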SoundCue /Game/Athena/Sounds/Ambient/Skytube/Skytube_Amb_Cue
root=Mixer_1
nodes [
 {id=Mixer_1, children [WavePlayer_1,WavePlayer_0], properties {inputVolume [1,1]}}
 {id=WavePlayer_0, properties {bLooping=true, soundWaveAssetPtr=&/Skytube_Energy_Layer_LP_01}}
 {id=WavePlayer_1, properties {bLooping=true, soundWaveAssetPtr=/Game/Athena/Sounds/Ambient/Skytube/Rev1/WindConduit_Ambient_01}}
]
properties {
 bOverrideAttenuation=true
 maxDistance=6050
 soundClassObject=&Fort_Ambient_Athena
 volumeMultiplier=1.253
 attenuationOverrides {
  bAttenuateWithLPF=true
  bEnableFocusInterpolation=true
  bEnableListenerFocus=true
  bEnableOcclusion=true
  falloffDistance=6000
  focusAzimuth=40
  focusDistanceScale=2
  focusReleaseInterpSpeed=5
  lPFFrequencyAtMax=6000
  lPFFrequencyAtMin=6000
  nonFocusVolumeAttenuation=0.1
  occlusionInterpolationTime=1
  occlusionLowPassFilterFrequency=2000
  occlusionTraceChannel=ECC_GameTraceChannel11
  occlusionVolumeAttenuation=0.5
  stereoSpread=3500
  attenuationShapeExtents {x=50}
 }
}
