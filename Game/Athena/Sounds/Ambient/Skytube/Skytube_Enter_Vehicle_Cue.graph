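# Skytube (wind conduit) vehicle-entry sound cue, playing one of four equally weighted WindConduit_Enter waves in the ambient sound class.
# An IsLocal switch selects either a default-attenuation path straight to the random pick, or the third-person Skytube attenuation path.
# The third-person path adds an IsVehicleEmpty switch that cuts volume to a quarter for empty vehicles.
# A root modulator wraps both paths.
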SoundCue /Game/Athena/Sounds/Ambient/Skytube/Skytube_Enter_Vehicle_Cue
root=Modulator_0
nodes [
 {
  id=Attenuation_4
  children [Switch_0]
  properties {
   attenuationSettings=&/Skytube_3P_Att
   attenuationOverrides {falloffDistance=10000, stereoSpread=1024, attenuationShapeExtents {x=0}}
  }
 }
 {id=Attenuation_5, children [Random_1]}
 {id=Mixer_0, children [Random_1], properties {inputVolume [1]}}
 {id=Mixer_1, children [Random_1], properties {inputVolume [0.25]}}
 {id=Modulator_0, children [Switch_1]}
 {id=Random_1, children [WavePlayer_5, WavePlayer_6, WavePlayer_7, WavePlayer_10], properties {weights [1,1,1,1]}}
 {id=Switch_0, children [Mixer_0,Mixer_0,Mixer_1], properties {intParameterName=IsVehicleEmpty}}
 {id=Switch_1, children [Attenuation_4, Attenuation_4, Attenuation_5], properties {intParameterName=IsLocal}}
 {id=WavePlayer_10, properties {soundWaveAssetPtr=/Game/Athena/Sounds/Ambient/Skytube/Rev1/WindConduit_Enter_07}}
 {id=WavePlayer_5, properties {soundWaveAssetPtr=/Game/Athena/Sounds/Ambient/Skytube/Rev1/WindConduit_Enter_01}}
 {id=WavePlayer_6, properties {soundWaveAssetPtr=/Game/Athena/Sounds/Ambient/Skytube/Rev1/WindConduit_Enter_03}}
 {id=WavePlayer_7, properties {soundWaveAssetPtr=/Game/Athena/Sounds/Ambient/Skytube/Rev1/WindConduit_Enter_06}}
]
properties {
 maxDistance=2097152
 soundClassObject=&Fort_Ambient_Athena
 volumeMultiplier=1.559
 attenuationOverrides {falloffDistance=6000, stereoSpread=2000, attenuationShapeExtents {x=0}}
}
