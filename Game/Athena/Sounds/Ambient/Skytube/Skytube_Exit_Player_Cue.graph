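# Sound cue for a player exiting a Skytube wind conduit in Athena, built from WindConduit "Enter" waves.
# An IsLocal switch sends unset or non-local cases to a boosted mixer over three random third-person waves with Skytube_3P_Att attenuation.
# The local case picks between two other waves through a separate attenuation node with no assigned settings.
# A root modulator wraps both branches, and the cue uses the Athena ambient sound class.
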SoundCue /Game/Athena/Sounds/Ambient/Skytube/Skytube_Exit_Player_Cue
root=Modulator_0
nodes [
 {
  id=Attenuation_0
  children [Random_1]
  properties {
   attenuationSettings=&/Skytube_3P_Att
   attenuationOverrides {falloffDistance=10000, stereoSpread=1024, attenuationShapeExtents {x=0}}
  }
 }
 {id=Attenuation_1, children [Random_5]}
 {id=Mixer_0, children [Attenuation_0], properties {inputVolume [1.5]}}
 {id=Modulator_0, children [Switch_0]}
 {id=Random_1, children [WavePlayer_3,WavePlayer_4,WavePlayer_5], properties {weights [1,1,1]}}
 {id=Random_5, children [WavePlayer_18, WavePlayer_20], properties {weights [1,1]}}
 {id=Switch_0, children [Mixer_0, Mixer_0, Attenuation_1], properties {intParameterName=IsLocal}}
 {id=WavePlayer_18, properties {soundWaveAssetPtr=/Game/Athena/Sounds/Ambient/Skytube/Rev1/WindConduit_Enter_02}}
 {id=WavePlayer_20, properties {soundWaveAssetPtr=/Game/Athena/Sounds/Ambient/Skytube/Rev1/WindConduit_Enter_04}}
 {id=WavePlayer_3, properties {soundWaveAssetPtr=/Game/Athena/Sounds/Ambient/Skytube/Rev1/WindConduit_Enter_Third_04}}
 {id=WavePlayer_4, properties {soundWaveAssetPtr=/Game/Athena/Sounds/Ambient/Skytube/Rev1/WindConduit_Enter_Third_05}}
 {id=WavePlayer_5, properties {soundWaveAssetPtr=/Game/Athena/Sounds/Ambient/Skytube/Rev1/WindConduit_Enter_Third_03}}
]
properties {
 maxDistance=2097152
 soundClassObject=&Fort_Ambient_Athena
 volumeMultiplier=1.247
 attenuationOverrides {falloffDistance=6000, stereoSpread=2000, attenuationShapeExtents {x=0}}
}
