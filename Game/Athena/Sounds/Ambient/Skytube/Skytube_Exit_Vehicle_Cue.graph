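# SoundCue that plays the sound of a vehicle exiting a Skytube, reusing WindConduit "Enter" wind waves in the ambient sound class.
# A root Modulator feeds an IsLocal switch: the local case randomizes two waves, while the unset and non-local cases use third-person waves with Skytube 3P attenuation.
# The third-person branch switches on IsVehicleEmpty, sending one shared three-wave random through a much quieter mixer when the vehicle is empty.
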SoundCue /Game/Athena/Sounds/Ambient/Skytube/Skytube_Exit_Vehicle_Cue
root=Modulator_0
nodes [
 {
  id=Attenuation_2
  children [Switch_0]
  properties {
   attenuationSettings=&/Skytube_3P_Att
   attenuationOverrides {falloffDistance=10000, stereoSpread=1024, attenuationShapeExtents {x=0}}
  }
 }
 {id=Attenuation_3, children [Random_5]}
 {id=Mixer_0, children [Random_0], properties {inputVolume [1.5]}}
 {id=Mixer_1, children [Random_0], properties {inputVolume [0.25]}}
 {id=Modulator_0, children [Switch_1]}
 {id=Random_0, children [WavePlayer_3,WavePlayer_0,WavePlayer_1], properties {weights [1,1,1]}}
 {id=Random_5, children [WavePlayer_18, WavePlayer_20], properties {weights [1,1]}}
 {id=Switch_0, children [Mixer_0,Mixer_0,Mixer_1], properties {intParameterName=IsVehicleEmpty}}
 {id=Switch_1, children [Attenuation_2, Attenuation_2, Attenuation_3], properties {intParameterName=IsLocal}}
 {id=WavePlayer_0, properties {soundWaveAssetPtr=/Game/Athena/Sounds/Ambient/Skytube/Rev1/WindConduit_Enter_Third_05}}
 {id=WavePlayer_1, properties {soundWaveAssetPtr=/Game/Athena/Sounds/Ambient/Skytube/Rev1/WindConduit_Enter_Third_03}}
 {id=WavePlayer_18, properties {soundWaveAssetPtr=/Game/Athena/Sounds/Ambient/Skytube/Rev1/WindConduit_Enter_02}}
 {id=WavePlayer_20, properties {soundWaveAssetPtr=/Game/Athena/Sounds/Ambient/Skytube/Rev1/WindConduit_Enter_04}}
 {id=WavePlayer_3, properties {soundWaveAssetPtr=/Game/Athena/Sounds/Ambient/Skytube/Rev1/WindConduit_Enter_Third_04}}
]
properties {
 maxDistance=2097152
 soundClassObject=&Fort_Ambient_Athena
 volumeMultiplier=1.247
 attenuationOverrides {falloffDistance=6000, stereoSpread=2000, attenuationShapeExtents {x=0}}
}
