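# Ambient sound cue for a Skytube ring in Athena, playing looping wind conduit audio in the ambient sound class.
# A random node picks one of two anchor loops, feeding smoothed Doppler, then an attenuation node with custom falloff, listener focus and occlusion.
# A mixer combines this with a half-volume Doppler and low-pass attenuation branch that has no wave connected, so it is silent.
# Concurrency overrides specify one instance.
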SoundCue /Game/Athena/Sounds/Ambient/Skytube/Skytube_Ring_Amb_Cue
root=Mixer_0
nodes [
 {
  id=Attenuation_1
  children [null]
  properties {
   bOverrideAttenuation=true
   attenuationOverrides {
    bAttenuateWithLPF=true
    bEnableOcclusion=true
    falloffDistance=2000
    hPFFrequencyAtMax=750
    lPFFrequencyAtMax=1000
    lPFFrequencyAtMin=1000
    lPFRadiusMax=100
    lPFRadiusMin=0
    occlusionInterpolationTime=1
    occlusionLowPassFilterFrequency=2000
    occlusionTraceChannel=ECC_GameTraceChannel11
    occlusionVolumeAttenuation=0.5
    attenuationShapeExtents {x=1000}
   }
  }
 }
 {
  id=Attenuation_2
  children [Doppler_0]
  properties {
   bOverrideAttenuation=true
   attenuationOverrides {
    bEnableFocusInterpolation=true
    bEnableListenerFocus=true
    bEnableOcclusion=true
    distanceAlgorithm=Custom
    falloffDistance=3500
    focusAttackInterpSpeed=5
    focusAzimuth=20
    focusDistanceScale=0.3
    focusReleaseInterpSpeed=0.25
    nonFocusAzimuth=30
    nonFocusVolumeAttenuation=0.3
    occlusionInterpolationTime=1
    occlusionLowPassFilterFrequency=2000
    occlusionTraceChannel=ECC_GameTraceChannel11
    occlusionVolumeAttenuation=0.5
    stereoSpread=2000
    attenuationShapeExtents {x=0}
    customAttenuationCurve {editorCurveData {keys [{arriveTangent=-2.7215, interpMode=RCIM_Cubic, leaveTangent=-2.7215, tangentMode=RCTM_User, time=0, value=1}, <at 1 0 RCIM_Cubic>]}}
   }
  }
 }
 {id=Doppler_0, children [Random_1], properties {bUseSmoothing=true, dopplerIntensity=2}}
 {id=Doppler_1, children [Attenuation_1], properties {bUseSmoothing=true, dopplerIntensity=2}}
 {id=Mixer_0, children [Doppler_1, Attenuation_2], properties {inputVolume [0.5,1]}}
 {id=Random_1, children [WavePlayer_1,WavePlayer_3], properties {weights [1,1]}}
 {id=WavePlayer_1, properties {bLooping=true, soundWaveAssetPtr=/Game/Athena/Sounds/Ambient/Skytube/Rev1/WindConduit_Anchor_02_opt}}
 {id=WavePlayer_3, properties {bLooping=true, soundWaveAssetPtr=/Game/Athena/Sounds/Ambient/Skytube/Rev1/WindConduit_Anchor_04_opt}}
]
properties {
 maxDistance=3500
 soundClassObject=&Fort_Ambient_Athena
 volumeMultiplier=2.079
 attenuationOverrides {
  bAttenuate=false
  bEnableFocusInterpolation=true
  bEnableListenerFocus=true
  bEnableOcclusion=true
  distanceAlgorithm=Custom
  falloffDistance=3500
  focusAttackInterpSpeed=5
  focusAzimuth=15
  focusDistanceScale=0.3
  focusReleaseInterpSpeed=0.25
  nonFocusAzimuth=25
  nonFocusVolumeAttenuation=0.2
  occlusionInterpolationTime=1
  occlusionLowPassFilterFrequency=2000
  occlusionTraceChannel=ECC_GameTraceChannel11
  occlusionVolumeAttenuation=0.5
  stereoSpread=2000
  attenuationShapeExtents {x=0}
  customAttenuationCurve {editorCurveData {keys [{arriveTangent=-3.2673, interpMode=RCIM_Cubic, leaveTangent=-3.2673, tangentMode=RCTM_User, time=0, value=1}, <at 1 0 RCIM_Cubic>]}}
 }
 concurrencyOverrides {bVolumeScaleCanRelease=true, maxCount=1, volumeScaleAttackTime=0.5}
}
