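# Sound cue for a Skytube ring being disabled, routed to the Athena ambient sound class with boosted volume.
# A single wave player feeds an Attenuation node that overrides attenuation with a custom falloff curve and listener focus, quieting the sound when out of focus.
# A concurrency override limits it to one instance, stopping the farthest before preventing new plays.
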SoundCue /Game/Athena/Sounds/Ambient/Skytube/Skytube_Ring_Disable_Cue
root=Attenuation_2
nodes [
 {
  id=Attenuation_2
  children [WavePlayer_5]
  properties {
   bOverrideAttenuation=true
   attenuationOverrides {
    bEnableFocusInterpolation=true
    bEnableListenerFocus=true
    distanceAlgorithm=Custom
    falloffDistance=7500
    focusAzimuth=20
    focusDistanceScale=0.3
    nonFocusAzimuth=30
    nonFocusVolumeAttenuation=0.5
    occlusionInterpolationTime=1
    occlusionLowPassFilterFrequency=2000
    occlusionTraceChannel=ECC_GameTraceChannel11
    occlusionVolumeAttenuation=0.5
    stereoSpread=1024
    attenuationShapeExtents {x=0}
    customAttenuationCurve {editorCurveData {keys [
       {arriveTangent=-2.7215, interpMode=RCIM_Cubic, leaveTangent=-2.7215, tangentMode=RCTM_User, time=0, value=1}
       {arriveTangent=-0.5052, interpMode=RCIM_Cubic, leaveTangent=-0.5052, tangentMode=RCTM_User, time=1, value=0}
      ]}}
   }
  }
 }
 {id=WavePlayer_5, properties {soundWaveAssetPtr=&/Skytube_Ring_Disable_01}}
]
properties {
 bOverrideConcurrency=true
 maxDistance=7500
 soundClassObject=&Fort_Ambient_Athena
 volumeMultiplier=4.158
 attenuationOverrides {
  bAttenuate=false
  bEnableFocusInterpolation=true
  bEnableListenerFocus=true
  bEnableOcclusion=true
  distanceAlgorithm=Custom
  falloffDistance=3500
  focusAttackInterpSpeed=5
  focusAzimuth=15
  focusDistanceScale=0.3
  focusReleaseInterpSpeed=0.25
  nonFocusAzimuth=25
  nonFocusVolumeAttenuation=0.2
  occlusionInterpolationTime=1
  occlusionLowPassFilterFrequency=2000
  occlusionTraceChannel=ECC_GameTraceChannel11
  occlusionVolumeAttenuation=0.5
  stereoSpread=2000
  attenuationShapeExtents {x=0}
  customAttenuationCurve {editorCurveData {keys [{arriveTangent=-3.2673, interpMode=RCIM_Cubic, leaveTangent=-3.2673, tangentMode=RCTM_User, time=0, value=1}, <at 1 0 RCIM_Cubic>]}}
 }
 concurrencyOverrides {bVolumeScaleCanRelease=true, maxCount=1, resolutionRule=StopFarthestThenPreventNew, volumeScale=0.5, volumeScaleAttackTime=0.5}
}
