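# Ambient SoundCue for the energy loop of a Skytube tunnel in UFO wreckage, routed to the Athena ambient sound class.
# It randomly picks one of three equally weighted looping waves, then applies slight pitch and volume modulation and a smoothed, reduced-intensity Doppler shift.
# Its attenuation override settings define distance falloff, low-pass filtering, listener focus, wide stereo spread and occlusion that muffles and quiets the sound.
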SoundCue /Game/Athena/Sounds/Ambient/Skytube/Skytube_Tunnel_UFO_Wreckage_Cue
root=Doppler_1
nodes [
 {id=Doppler_1, children [Modulator_1], properties {bUseSmoothing=true, dopplerIntensity=0.4}}
 {id=Modulator_1, children [Random_0], properties {pitchMax=1, pitchMin=0.98, volumeMax=1.2, volumeMin=1.1}}
 {id=Random_0, children [WavePlayer_2,WavePlayer_3,WavePlayer_4], properties {weights [1,1,1]}}
 {
  id=WavePlayer_2
  properties {bLooping=true, soundWaveAssetPtr=&/Skytube_Energy_UFO_Wreckage_Tunnel_1P_LP_01}
 }
 {
  id=WavePlayer_3
  properties {bLooping=true, soundWaveAssetPtr=&/Skytube_Energy_UFO_Wreckage_Tunnel_1P_LP_02}
 }
 {
  id=WavePlayer_4
  properties {bLooping=true, soundWaveAssetPtr=&/Skytube_Energy_UFO_Wreckage_Tunnel_1P_LP_03}
 }
]
properties {
 maxDistance=2097152
 soundClassObject=&Fort_Ambient_Athena
 volumeMultiplier=0.752
 attenuationOverrides {
  bAttenuateWithLPF=true
  bEnableFocusInterpolation=true
  bEnableListenerFocus=true
  bEnableOcclusion=true
  falloffDistance=6000
  focusAzimuth=40
  focusDistanceScale=2
  focusReleaseInterpSpeed=5
  lPFFrequencyAtMax=6000
  lPFFrequencyAtMin=6000
  nonFocusVolumeAttenuation=0.1
  occlusionInterpolationTime=1
  occlusionLowPassFilterFrequency=2000
  occlusionTraceChannel=ECC_GameTraceChannel11
  occlusionVolumeAttenuation=0.5
  stereoSpread=3500
  attenuationShapeExtents {x=50}
 }
}
